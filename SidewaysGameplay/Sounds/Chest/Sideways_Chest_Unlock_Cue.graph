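# Sound cue for the unlock sound of a chest in the Sideways gameplay content.
# A random node picks one of three equally weighted unlocking waves and feeds a single-input mixer at the root.
# It uses shared item-interact attenuation and a loot reward sound class, with a slight volume boost.
# A concurrency override prevents new instances from starting once the limit is reached.
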SoundCue /SidewaysGameplay/Sounds/Chest/Sideways_Chest_Unlock_Cue
root=Mixer_0
nodes [
 {id=Mixer_0, children [Random_0], properties {inputVolume [1]}}
 {id=Random_0, children [WavePlayer_2,WavePlayer_3,WavePlayer_4], properties {weights [1,1,1]}}
 {id=WavePlayer_2, properties {soundWaveAssetPtr=&/Sideways_Chest_Unlocking_01}}
 {id=WavePlayer_3, properties {soundWaveAssetPtr=&/Sideways_Chest_Unlocking_02}}
 {id=WavePlayer_4, properties {soundWaveAssetPtr=&/Sideways_Chest_Unlocking_03}}
]
properties {
 attenuationSettings=/Game/Sounds/Attenuation/Fort_Item_Interact_Att
 bOverrideConcurrency=true
 maxDistance=1850
 soundClassObject=&Fort_Loot_Reward_PSM
 volumeMultiplier=1.25
 concurrencyOverrides {resolutionRule=PreventNew}
}
